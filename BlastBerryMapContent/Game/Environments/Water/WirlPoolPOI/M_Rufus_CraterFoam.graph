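# Masked lit material for whirlpool crater foam, drawn as camera-facing sprites that pulse in size and height.
# A time sine, phase-offset by vertex alpha, drives scale and offset, with sprite size varying by camera distance.
# Sphere normals blend with flattened water normals, noise feeds two-tone color, and a distorted splash texture forms an erodible opacity mask.
# Dithered pixel depth offset softens intersections, except in low-quality and forward paths.
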Material /BlastBerryMapContent/Game/Environments/Water/WirlPoolPOI/M_Rufus_CraterFoam
graph {
 properties {
  bUsedWithMeshParticles=true
  bUsedWithNanite=true
  bUsedWithNiagaraMeshParticles=true
  blendMode=BLEND_Masked
  pixelDepthOffsetMode=PDOM_Legacy
  translucencyLightingMode=TLM_VolumetricPerVertexNonDirectional
  usedShadingModels=Default Lit
 }
 nodes [
  n1 = LinearInterpolate(
   A=ScalarParameter() {defaultValue=150, parameterName=SpriteScaleNear}
   B=ScalarParameter() {defaultValue=300, parameterName=SpriteScaleFar}
   Alpha=MaterialFunctionCall(Distance=ScalarParameter() {defaultValue=2500, parameterName=SoftenDistance}) {materialFunction=&Near_Camera_Fade}.Result
  )
  n2 = TextureSample(UVs=NamedRerouteUsage()) {samplerType=SAMPLERTYPE_Normal, texture=/Game/Packages/DS_Fortnite_TRV_Terrain/Textures/T_TRV_SwampWater_01_N}
  n3 = ScalarParameter() {defaultValue=0.7, parameterName=Flatten}
  n4 = MaterialFunctionCall!(
   Center Location=WorldPosition().XYZ
   0-1 UVs
   XY scale=AppendVector(A=@n1, B=@n1)
   OverridePivot
   "Normals (Optional)"=MaterialFunctionCall(
    BaseNormal=TextureSample() {samplerType=SAMPLERTYPE_Normal, texture=&T_Sphere_N}.RGB
    AdditionalNormal=MaterialFunctionCall(Normal=@n2.RGB, Flatness=@n3) {materialFunction=&FlattenNormal}.Result
   ) {materialFunction=&BlendAngleCorrectedNormals}.Result
  ) {materialFunction=&Sprite}
  n5 = Time()
  n6 = ScalarParameter() {defaultValue=0.8, parameterName=TimeMult}
  n7 = Frac(None=Add(A=VertexColor().A, B=Multiply(A=@n5, B=@n6) {constB=0.8}))
  n8 = NamedRerouteDeclaration(None=Sine(None=Divide(A=@n7))) {name=TimeSin, nodeColor=<rgba 1 0.0938 0 1>}
  n9 = Power(Base=@n8, Exp=ScalarParameter() {defaultValue=1, parameterName=ScalePower})
  n10 = VectorParameter() {parameterName=Offset_Low, defaultValue=<rgba 0 0 -150 0>}
  n11 = VectorParameter() {parameterName=Offset_High, defaultValue=<rgba 0 0 50 0>}
  n12 = Power(Base=@n8, Exp=ScalarParameter() {defaultValue=5, parameterName=OffsetPower}) {constExponent=1}
  n13 = Add(
   A=Multiply(A=@n4.World position offset, B=LinearInterpolate(Alpha=@n9) {constB=1.25})
   B=Transform(None=LinearInterpolate(A=@n10.RGB, B=@n11.RGB, Alpha=@n12)) {transformSourceType=TRANSFORMSOURCE_Local}
  )
  n14 = VertexColor()
  n15 = Time()
  n16 = Multiply(A=Multiply(A=Subtract(A=@n14.A) {constB=0.5}) {constB=2}, B=@n15) {constB=0.1}
  n17 = ScalarParameter() {defaultValue=-0.02}
  n18 = VertexColor()
  n19 = MaterialFunctionCall(R=@n18.R, G=@n18.R, B=@n18.R) {materialFunction=&AppendMany}
  Add(
   A=@n13
   B=RotateAboutAxis(
    NormalizedRotationAxis=MaterialFunctionCall!(Use World Position Offsets) {materialFunction=/Engine/Functions/Engine_MaterialFunctions02/Utility/CameraVectorWithWPOOptions}.Result
    RotationAngle=Multiply(A=@n16, B=@n17)
    PivotPoint=Transform(None=@n19.RGB) {transformSourceType=TRANSFORMSOURCE_World, transformType=TRANSFORM_Local}
    Position=@n13
   )
  )
  Frac(None=Add(A=@n14.A, B=@n15))
  n20 = Multiply(A=TextureCoordinate() {uTiling=0.9, vTiling=1.8}, B=ScalarParameter() {defaultValue=1, parameterName=ScaleOpacityMaskNoise})
  n21 = VertexColor()
  NamedRerouteDeclaration(None=Panner(Coordinate=Add(A=@n20, B=@n21.A), Speed=VectorParameter() {parameterName=DropSpeed, defaultValue=<rgba 0 -2 0 1>}.RGB) {speedY=-0.5}) {name=NoisePan, nodeColor=<rgba 0.8516 1 0 1>}
  n22 = TextureSample() {texture=/Game/Effects/Fort_Effects/Textures/T_Smoke_Soft}
  n23 = TextureSample(UVs=NamedRerouteUsage()) {samplerType=SAMPLERTYPE_LinearColor, texture=/Game/Weapons/WeaponSkins/Textures/T_Swirly_Noise_highContrast1}
  NamedRerouteDeclaration(None=LinearInterpolate(A=@n22.G, B=Multiply(A=@n22.G, B=@n23.G), Alpha=ScalarParameter() {defaultValue=0.5, parameterName=LErpoit})) {name=ColorNoise, nodeColor=<rgba 0 1 0.4844 1>}
  NamedRerouteDeclaration(None=LinearInterpolate(
    A=VectorParameter() {parameterName=ColorDim, defaultValue=<rgba 0.1904 0.5625 0.5625 1>}.RGB
    B=VectorParameter() {parameterName=ColorBright, defaultValue=<rgba 0.7782 0.9126 0.9167 1>}.RGB
    Alpha=NamedRerouteUsage()
   )) {name=BaseColor, nodeColor=<rgba 0.9688 0 1 1>}
  n24 = NamedRerouteUsage()
  NamedRerouteDeclaration(None=Multiply(
    A=ScalarParameter() {defaultValue=0.01, parameterName=Roughness}
    B=TextureSample(UVs=@n24) {samplerType=SAMPLERTYPE_LinearColor, texture=/Game/Global/Textures/Default/Masks_and_Gradients/T_Scratch_Mask}.G
   )) {name=Roughness, nodeColor=<rgba 0 1 0.3906 1>}
  n25 = NamedRerouteUsage()
  n26 = ScalarParameter() {defaultValue=110, parameterName=Offset}
  n27 = Constant()
  n28 = ShadingPathSwitch(
   Default=Multiply(
    A=Multiply(A=@n25, B=@n26)
    B=MaterialFunctionCall() {materialFunction=&DitherTemporalAA}.Result
   )
   Forward=@n27
  )
  NamedRerouteDeclaration(None=QualitySwitch(Default=@n28, Low=@n27, High=@n28, Medium=@n28, Epic=@n28)) {name=PixelDepthOffset, nodeColor=<rgba 0 1 0.2031 1>}
  NamedRerouteDeclaration(None=@n13) {name=WorldPosOffset, nodeColor=<rgba 1 0.5391 0 1>}
  NamedRerouteDeclaration(None=@n4.Normals) {name=Normal, nodeColor=<rgba 0.4766 1 0 1>}
  n29 = TextureSample(UVs=Panner(Coordinate=TextureCoordinate() {uTiling=2, vTiling=2}) {speedY=0.3}) {texture=/Game/Global/Textures/Default/Tileable_Patterns/T_Tilling_Distortion_Map_01}
  n30 = ComponentMask(None=@n29.G) {g=true, r=true}
  n31 = TextureSample(UVs=Add(A=TextureCoordinate() {coordinateIndex=1}, B=Multiply(A=@n30) {constB=0.1})) {samplerType=SAMPLERTYPE_LinearColor, texture=/Game/Effects/Fort_Effects/Textures/Water/T_waterSplash}
  n32 = Power(Base=NamedRerouteUsage(), Exp=ScalarParameter() {defaultValue=25, parameterName=ErosionPower}) {constExponent=5}
  n33 = Multiply(
   A=Power(Base=Multiply(A=@n31.G, B=@n29.G), Exp=LinearInterpolate(Alpha=@n32) {constA=5, constAlpha=0}) {constExponent=50}
   B=LinearInterpolate(Alpha=@n32) {constAlpha=0, constB=200}
  )
  NamedRerouteDeclaration(None=Multiply(A=@n31.G, B=Clamp(None=@n33))) {name=Opacity, nodeColor=<rgba 0.4766 1 0 1>}
  NamedRerouteUsage()
  NamedRerouteUsage()
  NamedRerouteUsage()
  NamedRerouteUsage()
  NamedRerouteUsage()
  NamedRerouteUsage()
  WorldPosition()
 ]
}
